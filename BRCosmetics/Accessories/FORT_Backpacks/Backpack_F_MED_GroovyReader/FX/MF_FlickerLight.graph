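# Material function for the GroovyReader backpack that outputs an irregular, time-driven flicker scalar, with parameters grouped under emissive random flicker.
# Six DebugTimeSine calls at separately tunable speeds build pseudo-random noise by lerping between a sum and a difference, using a product as alpha.
# The saturated noise is maxed against a scaled, bias-remapped sine with its own frequency parameter, so that smooth pulse acts as a moving floor under the flicker.
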MaterialFunction /BRCosmetics/Accessories/FORT_Backpacks/Backpack_F_MED_GroovyReader/FX/MF_FlickerLight
graph {nodes [
  n1 = ScalarParameter() {defaultValue=0.5, group=EmissiveRandomFlicker, parameterName=SineWaveMultiplierFrequency, sortPriority=36}
  n2 = Time()
  n3 = ConstantBiasScale(None=Sine(None=Multiply(A=@n1, B=@n2)))
  n4 = ScalarParameter() {defaultValue=3.2, group=EmissiveRandomFlicker, parameterName=RandomFlicker_Value01}
  n5 = ScalarParameter() {defaultValue=2, group=EmissiveRandomFlicker, parameterName=RandomFlicker_Value02}
  n6 = ScalarParameter() {defaultValue=0.23, group=EmissiveRandomFlicker, parameterName=RandomFlicker_Value03}
  n7 = ScalarParameter() {defaultValue=0.75, group=EmissiveRandomFlicker, parameterName=RandomFlicker_Value04}
  n8 = ScalarParameter() {defaultValue=0.33, group=EmissiveRandomFlicker, parameterName=RandomFlicker_Value05}
  n9 = ScalarParameter() {defaultValue=0.647, group=EmissiveRandomFlicker, parameterName=RandomFlicker_Value06}
  n10 = LinearInterpolate(
   A=Add(
    A=MaterialFunctionCall(Speed=@n4) {materialFunction=&DebugTimeSine}.Result
    B=MaterialFunctionCall(Speed=@n5) {materialFunction=&DebugTimeSine}.Result
   )
   B=Subtract(
    A=MaterialFunctionCall(Speed=@n6) {materialFunction=&DebugTimeSine}.Result
    B=MaterialFunctionCall(Speed=@n7) {materialFunction=&DebugTimeSine}.Result
   )
   Alpha=Multiply(
    A=MaterialFunctionCall(Speed=@n8) {materialFunction=&DebugTimeSine}.Result
    B=MaterialFunctionCall(Speed=@n9) {materialFunction=&DebugTimeSine}.Result
   )
  )
  FunctionOutput(None=Max(A=Multiply(A=@n3) {constB=0.885}, B=Saturate(None=@n10))) {outputName=Saturate}
 ]}
