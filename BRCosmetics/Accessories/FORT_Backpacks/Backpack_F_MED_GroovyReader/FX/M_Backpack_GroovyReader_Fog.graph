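# Translucent Niagara mesh-particle material that draws glowing, panning fog for the GroovyReader backpack cosmetic.
# A tiling cloud texture tinted by particle color blends toward a brighter tint inside a sphere mask offset from the particle by a dynamic parameter, alongside flicker logic.
# Opacity combines clouds, optional gradient, particle alpha, view-angle fade and vertex color, with depth fade skipped on forward and mobile.
# Dynamic parameters can drive tiling and pan.
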Material /BRCosmetics/Accessories/FORT_Backpacks/Backpack_F_MED_GroovyReader/FX/M_Backpack_GroovyReader_Fog
graph {
 properties {bUsedWithNiagaraMeshParticles=true, blendMode=BLEND_Translucent, pixelDepthOffsetMode=PDOM_Legacy, usedShadingModels=Default Lit}
 nodes [
  Constant() {r=10}
  n1 = TextureSample(UVs=NamedRerouteUsage()) {samplerType=SAMPLERTYPE_Grayscale, texture=&/T_fn_Fog_TilingClouds_D}
  n2 = ParticleColor()
  n3 = Multiply(A=@n1.RGB, B=@n2.RGB)
  n4 = Multiply(A=@n3, B=VectorParameter() {parameterName=SphereMaskColor, defaultValue=<rgba 3.24 9.6647 20 1>}.RGB) {constB=10}
  n5 = ParticlePositionWS()
  n6 = DynamicParameter() {parameterIndex=1, paramNames [X,Y,Z,ViewAngle]}
  n7 = Transform(None=@n6.RGB) {transformSourceType=TRANSFORMSOURCE_Instance}
  n8 = Multiply(
   A=SphereMask(
    A=WorldPosition().XYZ
    B=Add(A=@n5, B=@n7)
    Radius=ScalarParameter() {defaultValue=5, parameterName=Radius}
    Hardness=ScalarParameter() {defaultValue=-2, parameterName=Hardness}
   ) {hardnessPercent=0}
   B=NamedRerouteUsage()
  )
  Multiply(
   A=MaterialFunctionCall(Emissive=LinearInterpolate(A=@n3, B=@n4, Alpha=@n8) {constAlpha=0}) {materialFunction=&MF_FX_EmissiveAdjustments}.Result
   B=Constant() {r=0.6}
  )
  NamedRerouteDeclaration(None=MaterialFunctionCall() {materialFunction=&/MF_FlickerLight}.Saturate) {name=Flicker, nodeColor=<rgba 1 0 0.6562 1>}
  n9 = DynamicParameter() {defaultValue=<rgba 1 1 20 1>, paramNames [TileX,PanX,DepthFade,UseGradient]}
  NamedRerouteDeclaration(None=@n9.DepthFade) {name=DynParam1_DepthFade, nodeColor=<rgba 0 1 0.7891 1>}
  NamedRerouteDeclaration(None=@n9.TileX) {name=DynParam1_TileX, nodeColor=<rgba 1 0.7031 0 1>}
  NamedRerouteDeclaration(None=@n9.PanX) {name=DynParam1_PanX, nodeColor=<rgba 1 0.7031 0 1>}
  NamedRerouteDeclaration(None=@n9.UseGradient) {name=DynParam1_UseGradient, nodeColor=<rgba 0.1953 0 1 1>}
  n10 = TextureCoordinate()
  n11 = DynamicParameter() {defaultValue=<rgba 1 1 20 1>, paramNames [TileX,PanX,DepthFade,UseGradient]}
  n12 = ScalarParameter() {defaultValue=1, parameterName=UseDynamicParameters}
  n13 = AppendVector(
   A=LinearInterpolate(A=ScalarParameter() {defaultValue=0.5, parameterName=Tile_X}, B=@n11.TileX, Alpha=@n12)
   B=ScalarParameter() {defaultValue=0.2, parameterName=Tile_Y}
  )
  n14 = LinearInterpolate(A=ScalarParameter() {defaultValue=0.05, parameterName=Speed_X}, B=@n11.PanX, Alpha=@n12)
  n15 = ScalarParameter() {parameterName=Speed_Y}
  NamedRerouteDeclaration(None=Panner(Coordinate=Multiply(A=@n10, B=@n13) {constB=0.2}, Speed=AppendVector(A=@n14, B=@n15)) {speedX=0.1}) {name=UVs, nodeColor=<rgba 0.875 1 0 1>}
  NamedRerouteUsage()
  n16 = MaterialFunctionCall("FadeSensitivity (optional)"=ScalarParameter() {defaultValue=3, parameterName="FadeSensitivity (optional) (S)"}) {materialFunction=&FadeBasedOnViewAngle}
  OneMinus(None=@n16.Result)
  n17 = Multiply(
   A=@n1.R
   B=Power(
    Base=MaterialFunctionCall() {materialFunction=&LinearGradient}.VGradient
    Exp=ScalarParameter() {defaultValue=2.5, parameterName=Exp}
   )
  )
  n18 = NamedRerouteUsage()
  n19 = SmoothStep(Max=ScalarParameter() {defaultValue=1, parameterName=Max}, Value=Multiply(A=LinearInterpolate(A=@n1.R, B=@n17, Alpha=@n18), B=@n2.A))
  n20 = DepthFade(Opacity=@n19, FadeDistance=NamedRerouteUsage()) {fadeDistanceDefault=200}
  n21 = ShadingPathSwitch(Default=@n20, Deferred=@n20, Forward=@n19, Mobile=@n19)
  n22 = VertexColor()
  Reroute(None=Multiply(A=@n16.Result, B=Multiply(A=@n21, B=@n22)))
  ScalarParameter() {defaultValue=20, parameterName=FadeDistance}
  TransformPosition(None=@n6.RGB) {transformType=TRANSFORMPOSSOURCE_World}
 ]
}
