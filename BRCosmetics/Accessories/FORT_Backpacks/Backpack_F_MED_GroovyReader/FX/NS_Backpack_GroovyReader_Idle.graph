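# Niagara idle effect for the GroovyReader backpack cosmetic, adding blue-tinted atmospheric fog and drifting dust motes.
# Atmospheric_Fog bursts six persistent particles, using exec index as visibility tag across six mesh renderers and to pick per-layer color, material parameters and time-of-day brightness from arrays.
# Dust_Motes spawns box-placed sprites moved by wind, drag and curl noise, with two renderers and derived core size and color.
# User.Debug adds a burst and alters mote alpha.
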NiagaraSystem /BRCosmetics/Accessories/FORT_Backpacks/Backpack_F_MED_GroovyReader/FX/NS_Backpack_GroovyReader_Idle
dependencies {
 usedDataInterfaces [&NiagaraDataInterfaceArrayInt32, &NiagaraDataInterfaceCurve, &NiagaraDataInterfaceArrayFloat]
 usedDynamicInputs []
 usedModules [&SystemState]
}
emitters {
 Atmospheric_Fog {
  emitterUpdateScript [
   &EmitterState(
    Inactive Response=::NewEnumerator1
    Life Cycle Mode=::NewEnumerator1
    Loop Behavior=::NewEnumerator1
    Loop Duration Mode=::NewEnumerator1
    UseLoopDelay=false
   )
   &SpawnBurst_Instantaneous(Spawn Count=6, Use Spawn Probability=false, UseLoopCountLimit=false)
  ]
  particleSpawnScript [
   &InitializeParticle(Color=<rgba 0.0465 0.4341 1.5 8>, Position Mode=::NewEnumerator0, Sprite Size Mode=::NewEnumerator3, Write Material Random=true)
   &Color(Color=&MultiplyLinearColor(
     Linear Color=$Particles.Color
     Scale Factor=&MakeFloat4FromVectorAndFloat(
      "Vector (RGB)"=&VectorFromFloat(Value=&SelectFloatFromArray(
        Direct Array Index=&ReturnExecIndex()
        Float Selection Array=DataInterface() {propertyValues {FloatData [2,0.5,2,0.2,0.7,0.25]}}
       ))
      "Float (Alpha)"=&SelectFloatFromArray(
       Direct Array Index=&ReturnExecIndex()
       Float Selection Array=DataInterface() {propertyValues {FloatData [1,0.5,0.5,0.5,0.1,0.5]}}
      )
     )
    ))
   &DynamicMaterialParameters(
    Index 0 Param 1=&SelectFloatFromArray(
     Direct Array Index=&ReturnExecIndex()
     Float Selection Array=DataInterface() {propertyValues {FloatData [0.4,1,0.2,0.5,0.2,0.5]}}
    )
    Index 0 Param 2=&Multiply_Float(
     A=0.1
     B=&SelectFloatFromArray(
      Direct Array Index=&ReturnExecIndex()
      Float Selection Array=DataInterface() {propertyValues {FloatData [0.4,0.2,0.25,1,0.2,0.5]}}
     )
    )
    Index 0 Param 3=&SelectFloatFromArray(
     Direct Array Index=&ReturnExecIndex()
     Float Selection Array=DataInterface() {propertyValues {FloatData [4,2,20,1]}}
    )
    Index 0 Param 4=&SelectFloatFromArray(
     Direct Array Index=&ReturnExecIndex()
     Float Selection Array=DataInterface() {propertyValues {FloatData [1,1,1,0,0]}}
    )
    Index 1 Param 1=0.901
    Index 1 Param 2=-8.6041
    Index 1 Param 3=-4.6621
    Module.Index0Param0Write=true
    Module.Index0Param1Write=true
    Module.Index0Param2Write=true
    Module.Index0Param3Write=true
    Module.Index1Param0Write=true
    Module.Index1Param1Write=true
    Module.Index1Param2Write=true
    Module.Index1Param3Write=true
    Write Parameter Index 1=true
   )
   /SetVariables(Particles.VisibilityTag=&ReturnExecIndex()) {isSetParameters=true}
  ]
  particleUpdateScript [
   &ParticleState(Kill Particles When Lifetime Has Elapsed=false)
   &SolveForcesAndVelocity(Clamp Velocity=false, Limit Acceleration=false)
   &ScaleColor(
    Scale RGB=&VectorFromFloat(Value=&FNTimeOfDayScalar(
      Night Scale=&SelectFloatFromArray(
       Direct Array Index=&ReturnExecIndex()
       Float Selection Array=DataInterface() {propertyValues {FloatData [0.15,0.7,2,1.5,2,1]}}
      )
      Morning Scale=&SelectFloatFromArray(
       Direct Array Index=&ReturnExecIndex()
       Float Selection Array=DataInterface() {propertyValues {FloatData [0.2,0.7,1.5,2,4,2]}}
      )
      Day Scale=&SelectFloatFromArray(
       Direct Array Index=&ReturnExecIndex()
       Float Selection Array=DataInterface() {propertyValues {FloatData [0.25,0.7,2.5,4,4,2]}}
      )
      Evening Scale=&SelectFloatFromArray(
       Direct Array Index=&ReturnExecIndex()
       Float Selection Array=DataInterface() {propertyValues {FloatData [0.2,1,2.5,4,2,2]}}
      )
     ))
    ScaleA=true
    ScaleRGB=true
   )
  ]
  rendererClasses [&NiagaraMeshRendererProperties]
  renderers [
   {rendererClass=&NiagaraMeshRendererProperties, rendererIndex=0}
   {rendererClass=&NiagaraMeshRendererProperties, rendererIndex=1}
   {rendererClass=&NiagaraMeshRendererProperties, rendererIndex=2}
   {rendererClass=&NiagaraMeshRendererProperties, rendererIndex=3}
   {rendererClass=&NiagaraMeshRendererProperties, rendererIndex=4}
   {rendererClass=&NiagaraMeshRendererProperties, rendererIndex=5}
  ]
 }
 Dust_Motes {
  emitterUpdateScript [
   &EmitterState(Life Cycle Mode=::NewEnumerator1, Loop Duration=2, UseLoopDelay=false)
   &SpawnBurst_Instantaneous(
    Spawn Count=&SelectIntFromArray(
     Direct Array Index=&MakeIntFromFloat(Float=$User.Debug)
     Int Selection Array=DataInterface() {propertyValues {IntData [0,20]}}
    )
    Use Spawn Probability=false
    UseLoopCountLimit=false
   )
   &SpawnRate(SpawnRate=5, Use Spawn Probability=false)
  ]
  particleSpawnScript [
   &InitializeParticle(
    Color=<rgba 0.1105 0.3182 0.5 0.7>
    Lifetime Max=8
    Lifetime Min=5
    Lifetime Mode=::NewEnumerator1
    Position Offset=<0 -10 -5>
    Position Offset Coordinate Space=::Local
    Sprite Rotation Mode=::NewEnumerator2
    Sprite Size Mode=::NewEnumerator1
    Sprite UV Mode=::NewEnumerator3
    Uniform Sprite Size Max=3.5
    Uniform Sprite Size Min=0.5
    UsePositionOffset=true
    Write Material Random=true
   )
   &ShapeLocation(
    Box Midpoint=&RandomRangeVector(Minimum=<0 0 0>, Maximum=<1 1 1>)
    Box Size=<10 5 8>
    MaskSpawnGroup=false
    Override Seed=false
    Shape Primitive=::NewEnumerator2
    Surface Only=false
    UseOwnerScale=true
   )
  ]
  particleUpdateScript [
   &ParticleState()
   &WindForce(
    Frequency=20
    Noise Quality / Cost:ENiagaraCurlNoiseQuality=::NewEnumerator0
    Offset By Time:Vector3f=<0 0 -50>
    Scale=1
    Wind Speed=<0.1 0 1>
    Wind Speed Scale=&RandomRangeFloat(Minimum=0.05, Maximum=0.2)
   )
   &AerodynamicDrag(
    Aerodynamic Drag:NiagaraFloat=&RandomRangeFloat(Minimum=0.6, Maximum=1)
    Aerodynamic Rotational Drag=0.2
    Drag Shape Curve=DataInterface() {propertyValues {Curve {keys [
        {interpMode=RCIM_Cubic, tangentMode=RCTM_User, time=0, value=0.1}
        {arriveTangent=2.0799, interpMode=RCIM_Cubic, leaveTangent=2.0799, tangentMode=RCTM_User, time=1, value=1}
       ]}}}
    Lift Shape Curve=DataInterface() {propertyValues {Curve {keys [
        {interpMode=RCIM_Cubic, tangentMode=RCTM_User, time=0, value=0}
        {interpMode=RCIM_Cubic, tangentMode=RCTM_User, time=0.5, value=1}
        {tangentMode=RCTM_User, time=1, value=0}
       ]}}}
    Use Linear Drag=true
    Use Rotational Drag=true
   )
   &ScaleSpriteSize(Scale Factor=&Vector2DFromFloat(Value=&FloatFromCurve(
      FloatCurve=DataInterface() {propertyValues {Curve {keys [
          <at 0 0 RCIM_Cubic>
          <at 0.112 1.0173>
          {arriveTangent=-1.142, interpMode=RCIM_Cubic, leaveTangent=-1.142, time=0.7001, value=0.872}
          <at 1.0011 0.002>
         ]}}}
      Scale Curve=0.4
     )))
   &ScaleColor(
    Scale Alpha=&Lerp_Float(
     A=&FloatFromCurve(FloatCurve=DataInterface() {propertyValues {Curve {keys [<at 0 0 RCIM_Cubic>,<at 0.5 1 RCIM_Cubic>,<at 1 0 RCIM_Cubic>]}}})
     B=0.25
     Alpha=$User.Debug
    )
    ScaleA=true
    ScaleRGB=true
   )
   &CurlNoiseForce(Noise Frequency=400, Noise Strength=20, Pan Noise=false)
   &SolveForcesAndVelocity(Clamp Velocity=false, Limit Acceleration=false)
   /SetVariables(
    Particles.MoteCoreSize:Vector2f=&Multiply_Vector2D_ByFloat(Float=0.5, Vector2D=$Particles.SpriteSize)
    Particles.Motecolor:LinearColor=&MultiplyLinearColor(
     Linear Color=$Particles.Color
     Scale Factor=&MakeFloat4FromVectorAndFloat("Vector (RGB)"=<1 1 1>, "Float (Alpha)"=1.5)
    )
   ) {isSetParameters=true}
  ]
  rendererClasses [&NiagaraSpriteRendererProperties]
  renderers [
   {rendererClass=&NiagaraSpriteRendererProperties}
   {rendererClass=&NiagaraSpriteRendererProperties, rendererIndex=1}
  ]
 }
}
summary {
 systemName=NS_Backpack_GroovyReader_Idle
 emitters [
  {emitterName=Dust_Motes, rendererClasses [&NiagaraSpriteRendererProperties]}
  {emitterName=Atmospheric_Fog, rendererClasses [&NiagaraMeshRendererProperties]}
 ]
 userVariables [User.Debug {type=NiagaraFloat, defaultValue=0, description=""}]
}
