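# Physics asset for the GroovyReader backpack cosmetic, defining capsule bodies and joints for secondary motion on its rope and three small trees.
# Base bodies for the rope and each tree lock rotationally to the pack body, anchoring short chains of dyn bodies.
# The rope is a single dyn link, while each tree chains trunk to top.
# Tight swing and twist limits keep movement subtle, with the fl and fr trees swinging wider than br.
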PhysicsAsset /BRCosmetics/Accessories/FORT_Backpacks/Backpack_F_MED_GroovyReader/Meshes/F_MED_GroovyReader_Pack_Physics
bodies {
 base_rope [{shapeType=Capsule, length=0.5, radius=0.5, shapeName=base_rope_capsule}]
 base_tree_br [{shapeType=Capsule, length=0.5, radius=0.5, shapeName=base_tree_br_capsule}]
 base_tree_fl [{shapeType=Capsule, length=0.5, radius=0.5, shapeName=base_tree_fl_capsule}]
 base_tree_fr [{shapeType=Capsule, length=0.5, radius=0.5, shapeName=base_tree_fr_capsule}]
 dyn_rope [{shapeType=Capsule, length=1, radius=1, shapeName=dyn_rope_capsule, rotation=<rot 90 0 0>}]
 dyn_tree_br [{shapeType=Capsule, length=1, radius=1, shapeName=dyn_tree_br_capsule, center=<2 0 0>, rotation=<rot 90 0 0>}]
 dyn_tree_fl [{shapeType=Capsule, length=1, radius=1, shapeName=dyn_tree_fl_capsule, center=<2 0 0>, rotation=<rot 90 0 0>}]
 dyn_tree_fr [{shapeType=Capsule, length=1, radius=1, shapeName=dyn_tree_fr_capsule, center=<2 0 0>, rotation=<rot 90 0 0>}]
 dyn_treetrunk_br [{shapeType=Capsule, length=3, radius=0.5, shapeName=dyn_treetrunk_br_capsule, center=<1.5 0 0>, rotation=<rot 90 0 0>}]
 dyn_treetrunk_fl [{shapeType=Capsule, length=3, radius=0.5, shapeName=dyn_treetrunk_fl_capsule, center=<1.5 0 0>, rotation=<rot 90 0 0>}]
 dyn_treetrunk_fr [{shapeType=Capsule, length=3, radius=0.5, shapeName=dyn_treetrunk_fr_capsule, center=<1.5 0 0>, rotation=<rot 90 0 0>}]
 pack [{shapeType=Capsule, length=0.5, radius=0.5, shapeName=pack_capsule, rotation=<rot 90 0 0>}]
}
constraints [
 {
  bone1Name=base_rope
  bone2Name=pack
  swing1LimitDegrees=45
  swing1Motion=Locked
  swing2LimitDegrees=45
  swing2Motion=Locked
  twistLimitDegrees=45
  twistMotion=Locked
 }
 {bone1Name=dyn_rope, bone2Name=base_rope, swing1LimitDegrees=2, swing2LimitDegrees=2, twistLimitDegrees=1}
 {
  bone1Name=base_tree_br
  bone2Name=pack
  swing1LimitDegrees=45
  swing1Motion=Locked
  swing2LimitDegrees=45
  swing2Motion=Locked
  twistLimitDegrees=45
  twistMotion=Locked
 }
 {bone1Name=dyn_treetrunk_br, bone2Name=base_tree_br, swing1LimitDegrees=2, swing2LimitDegrees=2, twistLimitDegrees=1}
 {bone1Name=dyn_tree_br, bone2Name=dyn_treetrunk_br, swing1LimitDegrees=2, swing2LimitDegrees=2, twistLimitDegrees=0}
 {
  bone1Name=base_tree_fl
  bone2Name=pack
  swing1LimitDegrees=45
  swing1Motion=Locked
  swing2LimitDegrees=45
  swing2Motion=Locked
  twistLimitDegrees=45
  twistMotion=Locked
 }
 {bone1Name=dyn_treetrunk_fl, bone2Name=base_tree_fl, swing1LimitDegrees=4, swing2LimitDegrees=3, twistLimitDegrees=1}
 {bone1Name=dyn_tree_fl, bone2Name=dyn_treetrunk_fl, swing1LimitDegrees=10, swing2LimitDegrees=3, twistLimitDegrees=1}
 {
  bone1Name=base_tree_fr
  bone2Name=pack
  swing1LimitDegrees=45
  swing1Motion=Locked
  swing2LimitDegrees=45
  swing2Motion=Locked
  twistLimitDegrees=45
  twistMotion=Locked
 }
 {bone1Name=dyn_treetrunk_fr, bone2Name=base_tree_fr, swing1LimitDegrees=4, swing2LimitDegrees=3, twistLimitDegrees=0}
 {bone1Name=dyn_tree_fr, bone2Name=dyn_treetrunk_fr, swing1LimitDegrees=10, swing2LimitDegrees=3, twistLimitDegrees=1}
]
