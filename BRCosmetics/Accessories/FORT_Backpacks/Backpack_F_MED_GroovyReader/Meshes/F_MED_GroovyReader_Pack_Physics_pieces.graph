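# Physics asset for the GroovyReader backpack cosmetic, giving its car, cowboy, house and squirrel pieces secondary motion.
# Capsule bodies cover the pack, a small base bone per piece and a dyn bone per piece, with the house using two capsules.
# Each base bone is angularly locked to the pack, and each dyn bone hangs off its base with tight swing and twist limits.
# The squirrel swings widest, while the house barely moves and has twist locked.
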PhysicsAsset /BRCosmetics/Accessories/FORT_Backpacks/Backpack_F_MED_GroovyReader/Meshes/F_MED_GroovyReader_Pack_Physics_pieces
bodies {
 base_car [{shapeType=Capsule, length=0.5, radius=0.5, shapeName=base_car_capsule}]
 base_cowboy [{shapeType=Capsule, length=0.5, radius=0.5, shapeName=base_cowboy_capsule}]
 base_house [{shapeType=Capsule, length=0.5, radius=0.5, shapeName=base_house_capsule}]
 base_squirrel [{shapeType=Capsule, length=0.5, radius=0.5, shapeName=base_squirrel_capsule}]
 dyn_car [{shapeType=Capsule, length=2, radius=2, shapeName=dyn_car_capsule, rotation=<rot 0 0 90>}]
 dyn_cowboy [{shapeType=Capsule, length=3, radius=1, shapeName=dyn_cowboy_capsule, center=<2 0 0>, rotation=<rot 90 180 180>}]
 dyn_house [
  {shapeType=Capsule, length=3.5, radius=3.5, shapeName=dyn_house_capsule, center=<3.5 0 0>}
  {shapeType=Capsule, length=2, radius=1, shapeName=dyn_house_capsule, center=<2.4265 -4.4 -3>}
 ]
 dyn_squirrel [{shapeType=Capsule, length=1, radius=1, shapeName=dyn_squirrel_capsule, center=<0.5 0 0>, rotation=<rot 90 0 0>}]
 pack [{shapeType=Capsule, length=0.5, radius=0.5, shapeName=pack_capsule, rotation=<rot 90 0 0>}]
}
constraints [
 {
  bone1Name=base_car
  bone2Name=pack
  swing1LimitDegrees=45
  swing1Motion=Locked
  swing2LimitDegrees=45
  swing2Motion=Locked
  twistLimitDegrees=45
  twistMotion=Locked
 }
 {bone1Name=dyn_car, bone2Name=base_car, swing1LimitDegrees=2, swing2LimitDegrees=5, twistLimitDegrees=1}
 {
  bone1Name=base_cowboy
  bone2Name=pack
  swing1LimitDegrees=45
  swing1Motion=Locked
  swing2LimitDegrees=45
  swing2Motion=Locked
  twistLimitDegrees=45
  twistMotion=Locked
 }
 {bone1Name=dyn_cowboy, bone2Name=base_cowboy, swing1LimitDegrees=7, swing2LimitDegrees=7, twistLimitDegrees=1}
 {
  bone1Name=base_squirrel
  bone2Name=pack
  swing1LimitDegrees=45
  swing1Motion=Locked
  swing2LimitDegrees=45
  swing2Motion=Locked
  twistLimitDegrees=45
  twistMotion=Locked
 }
 {bone1Name=dyn_squirrel, bone2Name=base_squirrel, swing1LimitDegrees=10, swing2LimitDegrees=20, twistLimitDegrees=1}
 {
  bone1Name=base_house
  bone2Name=pack
  swing1LimitDegrees=45
  swing1Motion=Locked
  swing2LimitDegrees=45
  swing2Motion=Locked
  twistLimitDegrees=45
  twistMotion=Locked
 }
 {bone1Name=dyn_house, bone2Name=base_house, swing1LimitDegrees=2, swing2LimitDegrees=1, twistLimitDegrees=0, twistMotion=Locked}
]
